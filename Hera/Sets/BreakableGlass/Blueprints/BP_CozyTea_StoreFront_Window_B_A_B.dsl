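; BP_CozyTea_StoreFront_Window_B_A_B is a Blueprint for a Cozy Tea storefront window in the breakable glass set.
; Its Construction Script calls ValidateGlassProviderComponents on itself, then runs the parent class's construction script.
; The file holds no other logic of its own.
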(fn ConstructionScript
  (SDP|ValidateGlassProviderComponents self)
  (|CallParentFunction))
